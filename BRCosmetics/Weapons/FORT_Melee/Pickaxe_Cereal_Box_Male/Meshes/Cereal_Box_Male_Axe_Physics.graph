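# Physics asset for the Cereal Box Male pickaxe, with bodies for its dyn_ cereal and milk bones.
# Five small capsule bodies for cereal pieces and one for milk each constrain directly to dyn_cereal_A, a hub body built from four larger capsules.
# The milk joint locks swing and twist, while the cereal joints leave angular motion at default and carry differing limit angles.
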PhysicsAsset /BRCosmetics/Weapons/FORT_Melee/Pickaxe_Cereal_Box_Male/Meshes/Cereal_Box_Male_Axe_Physics
bodies {
 dyn_cereal_1 [{shapeType=Capsule, length=0.9248, radius=4.7559, shapeName=dyn_cereal_1_capsule}]
 dyn_cereal_2 [{shapeType=Capsule, length=0, radius=4.2527, shapeName=dyn_cereal_2_capsule, center=<1.6503 2.3377 0>}]
 dyn_cereal_3 [{shapeType=Capsule, length=0.0001, radius=4.7997, shapeName=dyn_cereal_3_capsule}]
 dyn_cereal_4 [{shapeType=Capsule, length=0.3621, radius=4.6988, shapeName=dyn_cereal_4_capsule}]
 dyn_cereal_5 [{shapeType=Capsule, length=0.0001, radius=3.3004, shapeName=dyn_cereal_5_capsule, center=<0.0115 0 0>}]
 dyn_cereal_A [
  {
   shapeType=Capsule
   length=24.9171
   radius=89.4223
   shapeName=dyn_cereal_A_capsule
   center=<10.7048 91.2704 8.7597>
   rotation=<rot -72.2062 65.8222 -72.0418>
  }
  {
   shapeType=Capsule
   length=40.221
   radius=15
   shapeName=dyn_cereal_A_capsule
   center=<0.3639 9.9669 24.8798>
   rotation=<rot 77.5208 86.2097 86.1181>
  }
  {
   shapeType=Capsule
   length=29.5712
   radius=15
   shapeName=dyn_cereal_A_capsule
   center=<-31.2648 3.6446 3.5146>
   rotation=<rot -22.1687 3.0118 -2.9521>
  }
  {
   shapeType=Capsule
   length=25.703
   radius=15
   shapeName=dyn_cereal_A_capsule
   center=<5.4551 12.462 20.0146>
   rotation=<rot -73.9561 -179.9999 173.7301>
  }
 ]
 dyn_milk_1 [{shapeType=Capsule, length=0, radius=6.5874, shapeName=dyn_milk_1_capsule, center=<2.1631 0 0>, rotation=<rot 0 -90 0>}]
}
constraints [
 {bone1Name=dyn_cereal_1, bone2Name=dyn_cereal_A, swing1LimitDegrees=90, swing2LimitDegrees=90, twistLimitDegrees=90}
 {bone1Name=dyn_cereal_2, bone2Name=dyn_cereal_A, swing1LimitDegrees=120, swing2LimitDegrees=120, twistLimitDegrees=120}
 {bone1Name=dyn_cereal_3, bone2Name=dyn_cereal_A, swing1LimitDegrees=45, swing2LimitDegrees=45, twistLimitDegrees=45}
 {bone1Name=dyn_cereal_4, bone2Name=dyn_cereal_A, swing1LimitDegrees=90, swing2LimitDegrees=90, twistLimitDegrees=90}
 {bone1Name=dyn_cereal_5, bone2Name=dyn_cereal_A, swing1LimitDegrees=90, swing2LimitDegrees=90, twistLimitDegrees=90}
 {
  bone1Name=dyn_milk_1
  bone2Name=dyn_cereal_A
  swing1LimitDegrees=45
  swing1Motion=Locked
  swing2LimitDegrees=45
  swing2Motion=Locked
  twistLimitDegrees=45
  twistMotion=Locked
 }
]
