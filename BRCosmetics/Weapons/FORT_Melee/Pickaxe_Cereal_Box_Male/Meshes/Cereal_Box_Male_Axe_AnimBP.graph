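# Animation Blueprint for the cereal box pickaxe (male), adding physics-driven secondary motion to its cereal rings.
# Two RigidBody passes, weighted for gameplay or front end, feed nine spring-constrained AnimDynamics ring bones.
# Cached poses pass that result through additive bone offsets on ring bases and ends to the output.
# The event graph sets front-end weights and sim-space settings, computes sine-wave ring rotators, and tracks jumping.
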AnimBlueprint /BRCosmetics/Weapons/FORT_Melee/Pickaxe_Cereal_Box_Male/Meshes/Cereal_Box_Male_Axe_AnimBP
target_skeleton=&/Cereal_Box_Male_Axe_Skeleton
graphs [
 AnimGraph {nodes [
   Output Pose(Result=Use cached pose 'CR_Cache_Transforms'()) {layerGroup=None}
   n1 = RigidBody(ComponentPose=Local To Component(LocalPose=Local Space Ref Pose()), Alpha=Get isNotFrontEnd() -> (isNotFrontEnd:double)) {
    bOverrideWorldGravity=true
    overridePhysicsAsset=&/Cereal_Box_Male_Axe_Physics
    simulationSpace=BaseBoneSpace
    alphaBoolBlend {blendOption=Linear}
    baseBoneRef {boneName=b_AR_Root}
    componentAppliedLinearAccClamp=<1000 2000 1000>
    componentLinearAccScale=<0.2 0.85 0.8>
    componentLinearVelScale=<0.2 0.85 0.8>
    overrideWorldGravity {z=-980}
   }
   n2 = RigidBody(ComponentPose=@n1, Alpha=Get isFrontEnd() -> (isFrontEnd:double)) {
    bOverrideWorldGravity=true
    overridePhysicsAsset=&Cereal_Box_Male_Axe_Physics
    simulationSpace=BaseBoneSpace
    alphaBoolBlend {blendOption=Linear}
    baseBoneRef {boneName=b_AR_Root}
    componentAppliedLinearAccClamp=<1000 1000 1000>
    componentLinearAccScale=<0.5 0.5 1>
    componentLinearVelScale=<0.5 0.5 1>
    overrideWorldGravity {x=100, z=-500}
   }
   Save cached pose 'CR_Cache_Input'(Pose=Component To Local(ComponentPose=@n2)) {cacheName=CR_Cache_Input}
   n3 = Anim Dynamics - Bone: dyn_cereal_ring_1!(
    ComponentPose=Local To Component(LocalPose=Use cached pose 'CR_Cache_Input'())
    Alpha=1
    SimSpaceSettings:Anim Phys Sim Space Settings Structure=(SimSpaceAngularAlpha=0,MaxAngularVelocity=10000,MaxAngularAcceleration=10000,ExternalAngularVelocity=(X=0,Y=0,Z=0))
    ExternalForce="0,0,0"
   ) -> (Pose) {
    angularDampingOverride=1
    angularSpringConstant=20000
    bAngularSpring=true
    bOverrideAngularDamping=true
    bUseGravityOverride=true
    gravityScale=0.35
    simulationSpace=RootRelative
    boundBone {boneName=dyn_cereal_ring_1}
    componentAppliedLinearAccClamp=<10000 10000 10000>
    componentLinearAccScale=<1 1 1>
    componentLinearVelScale=<1 1 1>
    gravityOverride {z=-980}
    physicsBodyDefinitions [{
      boundBone {boneName=dyn_cereal_ring_1}
      boxExtents=<30 40 40>
      constraintSetup {angularLimitsMax=<5 0 10>, angularLimitsMin=<-5 0 -10>, angularTarget=<1 0 0>}
      localJointOffset=<6 0 0>
     }]
   }
   n4 = Anim Dynamics - Bone: dyn_cereal_ring_2!(
    ComponentPose=@n3
    Alpha=1
    SimSpaceSettings:Anim Phys Sim Space Settings Structure=(SimSpaceAngularAlpha=0,MaxAngularVelocity=10000,MaxAngularAcceleration=10000,ExternalAngularVelocity=(X=0,Y=0,Z=0))
    ExternalForce="0,0,0"
   ) -> (Pose) {
    angularDampingOverride=1
    angularSpringConstant=20000
    bAngularSpring=true
    bOverrideAngularDamping=true
    bUseGravityOverride=true
    gravityScale=0.35
    simulationSpace=RootRelative
    boundBone {boneName=dyn_cereal_ring_2}
    componentAppliedLinearAccClamp=<10000 10000 10000>
    componentLinearAccScale=<1 1 1>
    componentLinearVelScale=<1 1 1>
    gravityOverride {z=-980}
    physicsBodyDefinitions [{
      boundBone {boneName=dyn_cereal_ring_2}
      boxExtents=<30 40 40>
      constraintSetup {angularLimitsMax=<3 0 3>, angularLimitsMin=<-3 0 -10>, angularTarget=<1 0 0>}
      localJointOffset=<6 0 0>
     }]
   }
   n5 = Anim Dynamics - Bone: dyn_cereal_ring_3!(
    ComponentPose=@n4
    Alpha=1
    SimSpaceSettings:Anim Phys Sim Space Settings Structure=(SimSpaceAngularAlpha=0,MaxAngularVelocity=10000,MaxAngularAcceleration=10000,ExternalAngularVelocity=(X=0,Y=0,Z=0))
    ExternalForce="0,0,0"
   ) -> (Pose) {
    angularDampingOverride=1
    angularSpringConstant=20000
    bAngularSpring=true
    bOverrideAngularDamping=true
    bUseGravityOverride=true
    gravityScale=0.35
    simulationSpace=RootRelative
    boundBone {boneName=dyn_cereal_ring_3}
    componentAppliedLinearAccClamp=<10000 10000 10000>
    componentLinearAccScale=<1 1 1>
    componentLinearVelScale=<1 1 1>
    gravityOverride {z=-980}
    physicsBodyDefinitions [{
      boundBone {boneName=dyn_cereal_ring_3}
      boxExtents=<30 40 40>
      constraintSetup {angularLimitsMax=<6 0 10>, angularLimitsMin=<-6 0 -15>, angularTarget=<1 0 0>}
      localJointOffset=<6 0 0>
     }]
   }
   n6 = Anim Dynamics - Bone: dyn_cereal_ring_4!(
    ComponentPose=@n5
    Alpha=1
    SimSpaceSettings:Anim Phys Sim Space Settings Structure=(SimSpaceAngularAlpha=0,MaxAngularVelocity=10000,MaxAngularAcceleration=10000,ExternalAngularVelocity=(X=0,Y=0,Z=0))
    ExternalForce="0,0,0"
   ) -> (Pose) {
    angularDampingOverride=1
    angularSpringConstant=20000
    bAngularSpring=true
    bOverrideAngularDamping=true
    bUseGravityOverride=true
    gravityScale=0.35
    simulationSpace=RootRelative
    boundBone {boneName=dyn_cereal_ring_4}
    componentAppliedLinearAccClamp=<10000 10000 10000>
    componentLinearAccScale=<1 1 1>
    componentLinearVelScale=<1 1 1>
    gravityOverride {z=-980}
    physicsBodyDefinitions [{
      boundBone {boneName=dyn_cereal_ring_4}
      boxExtents=<30 40 40>
      constraintSetup {angularLimitsMax=<4 0 10>, angularLimitsMin=<-4 0 -5>, angularTarget=<1 0 0>}
      localJointOffset=<6 0 0>
     }]
   }
   n7 = Anim Dynamics - Bone: dyn_cereal_ring_5!(
    ComponentPose=@n6
    Alpha=1
    SimSpaceSettings:Anim Phys Sim Space Settings Structure=(SimSpaceAngularAlpha=0,MaxAngularVelocity=10000,MaxAngularAcceleration=10000,ExternalAngularVelocity=(X=0,Y=0,Z=0))
    ExternalForce="0,0,0"
   ) -> (Pose) {
    angularDampingOverride=1
    angularSpringConstant=20000
    bAngularSpring=true
    bOverrideAngularDamping=true
    bUseGravityOverride=true
    gravityScale=0.35
    simulationSpace=RootRelative
    boundBone {boneName=dyn_cereal_ring_5}
    componentAppliedLinearAccClamp=<10000 10000 10000>
    componentLinearAccScale=<1 1 1>
    componentLinearVelScale=<1 1 1>
    gravityOverride {z=-980}
    physicsBodyDefinitions [{
      boundBone {boneName=dyn_cereal_ring_5}
      boxExtents=<30 40 40>
      constraintSetup {angularLimitsMax=<2 0 10>, angularLimitsMin=<-2 0 -2>, angularTarget=<1 0 0>}
      localJointOffset=<6 0 0>
     }]
   }
   n8 = Anim Dynamics - Bone: dyn_cereal_ring_6!(
    ComponentPose=@n7
    Alpha=1
    SimSpaceSettings:Anim Phys Sim Space Settings Structure=(SimSpaceAngularAlpha=0,MaxAngularVelocity=10000,MaxAngularAcceleration=10000,ExternalAngularVelocity=(X=0,Y=0,Z=0))
    ExternalForce="0,0,0"
   ) -> (Pose) {
    angularDampingOverride=1
    angularSpringConstant=20000
    bAngularSpring=true
    bOverrideAngularDamping=true
    bUseGravityOverride=true
    gravityScale=0.35
    simulationSpace=RootRelative
    boundBone {boneName=dyn_cereal_ring_6}
    componentAppliedLinearAccClamp=<10000 10000 10000>
    componentLinearAccScale=<1 1 1>
    componentLinearVelScale=<1 1 1>
    gravityOverride {z=-980}
    physicsBodyDefinitions [{
      boundBone {boneName=dyn_cereal_ring_6}
      boxExtents=<30 40 40>
      constraintSetup {angularLimitsMax=<8 0 5>, angularLimitsMin=<-8 0 -5>, angularTarget=<1 0 0>}
      localJointOffset=<6 0 0>
     }]
   }
   n9 = Anim Dynamics - Bone: dyn_cereal_ring_7!(
    ComponentPose=@n8
    Alpha=1
    SimSpaceSettings:Anim Phys Sim Space Settings Structure=(SimSpaceAngularAlpha=0,MaxAngularVelocity=10000,MaxAngularAcceleration=10000,ExternalAngularVelocity=(X=0,Y=0,Z=0))
    ExternalForce="0,0,0"
   ) -> (Pose) {
    angularDampingOverride=1
    angularSpringConstant=20000
    bAngularSpring=true
    bOverrideAngularDamping=true
    bUseGravityOverride=true
    gravityScale=0.35
    simulationSpace=RootRelative
    boundBone {boneName=dyn_cereal_ring_7}
    componentAppliedLinearAccClamp=<10000 10000 10000>
    componentLinearAccScale=<1 1 1>
    componentLinearVelScale=<1 1 1>
    gravityOverride {z=-980}
    physicsBodyDefinitions [{
      boundBone {boneName=dyn_cereal_ring_7}
      boxExtents=<30 40 40>
      constraintSetup {angularLimitsMax=<6 0 10>, angularLimitsMin=<-6 0 -10>, angularTarget=<1 0 0>}
      localJointOffset=<6 0 0>
     }]
   }
   n10 = Anim Dynamics - Bone: dyn_cereal_ring_8!(
    ComponentPose=@n9
    Alpha=1
    SimSpaceSettings:Anim Phys Sim Space Settings Structure=(SimSpaceAngularAlpha=0,MaxAngularVelocity=10000,MaxAngularAcceleration=10000,ExternalAngularVelocity=(X=0,Y=0,Z=0))
    ExternalForce="0,0,0"
   ) -> (Pose) {
    angularDampingOverride=1
    angularSpringConstant=20000
    bAngularSpring=true
    bOverrideAngularDamping=true
    bUseGravityOverride=true
    gravityScale=0.35
    simulationSpace=RootRelative
    boundBone {boneName=dyn_cereal_ring_8}
    componentAppliedLinearAccClamp=<10000 10000 10000>
    componentLinearAccScale=<1 1 1>
    componentLinearVelScale=<1 1 1>
    gravityOverride {z=-980}
    physicsBodyDefinitions [{
      boundBone {boneName=dyn_cereal_ring_8}
      boxExtents=<30 40 40>
      constraintSetup {angularLimitsMax=<3 0 5>, angularLimitsMin=<-3 0 -10>, angularTarget=<1 0 0>}
      localJointOffset=<6 0 0>
     }]
   }
   n11 = Anim Dynamics - Bone: dyn_cereal_ring_9!(
    ComponentPose=@n10
    Alpha=1
    SimSpaceSettings:Anim Phys Sim Space Settings Structure=(SimSpaceAngularAlpha=0,MaxAngularVelocity=10000,MaxAngularAcceleration=10000,ExternalAngularVelocity=(X=0,Y=0,Z=0))
    ExternalForce="0,0,0"
   ) -> (Pose) {
    angularDampingOverride=1
    angularSpringConstant=20000
    bAngularSpring=true
    bOverrideAngularDamping=true
    bUseGravityOverride=true
    gravityScale=0.35
    simulationSpace=RootRelative
    boundBone {boneName=dyn_cereal_ring_9}
    componentAppliedLinearAccClamp=<10000 10000 10000>
    componentLinearAccScale=<1 1 1>
    componentLinearVelScale=<1 1 1>
    gravityOverride {z=-980}
    physicsBodyDefinitions [{
      boundBone {boneName=dyn_cereal_ring_9}
      boxExtents=<30 40 40>
      constraintSetup {angularLimitsMax=<3 0 10>, angularLimitsMin=<-3 0 -10>, angularTarget=<1 0 0>}
      localJointOffset=<6 0 0>
     }]
   }
   Save cached pose 'CR_Cache_Dynamics'(Pose=Component To Local(ComponentPose=@n11)) {cacheName=CR_Cache_Dynamics}
   n12 = "Transform (Modify) Bone - Bone: base_cereal_ring_1"!(
    ComponentPose=Local To Component(LocalPose=Use cached pose 'CR_Cache_Dynamics'())
    bAlphaBoolEnabled:bool=True
    Translation="1.991,0.494,5.084"
    Rotation="-30,0,0"
    Scale="1,1,1"
   ) -> (Pose) {
    alphaInputType=Bool
    rotationMode=BMM_Additive
    rotationSpace=BCS_BoneSpace
    translationMode=BMM_Additive
    translationSpace=BCS_BoneSpace
    alphaBoolBlend {blendInTime=1, blendOutTime=0.1}
    alphaScaleBiasClamp {bInterpResult=true, interpSpeedDecreasing=5, interpSpeedIncreasing=5}
    boneToModify {boneName=base_cereal_ring_1}
    rotation {pitch=-30}
    translation=<1.991 0.494 5.084>
   }
   n13 = "Transform (Modify) Bone - Bone: base_cereal_ring_2"!(ComponentPose=@n12, bAlphaBoolEnabled:bool=True, Translation="-0.214,0.014,6.232", Rotation="-20,0,0", Scale="1,1,1") -> (Pose) {
    alphaInputType=Bool
    rotationMode=BMM_Additive
    rotationSpace=BCS_BoneSpace
    translationMode=BMM_Additive
    translationSpace=BCS_BoneSpace
    alphaBoolBlend {blendInTime=1, blendOutTime=0.1}
    alphaScaleBiasClamp {bInterpResult=true, interpSpeedDecreasing=5, interpSpeedIncreasing=5}
    boneToModify {boneName=base_cereal_ring_2}
    rotation {pitch=-20}
    translation=<-0.214 0.014 6.232>
   }
   n14 = "Transform (Modify) Bone - Bone: base_cereal_ring_3"!(ComponentPose=@n13, bAlphaBoolEnabled:bool=True, Translation="1.002,-0.000,2.230", Rotation="-25,0,0", Scale="1,1,1") -> (Pose) {
    alphaInputType=Bool
    rotationMode=BMM_Additive
    rotationSpace=BCS_BoneSpace
    translationMode=BMM_Additive
    translationSpace=BCS_BoneSpace
    alphaBoolBlend {blendInTime=1, blendOutTime=0.1}
    alphaScaleBiasClamp {bInterpResult=true, interpSpeedDecreasing=5, interpSpeedIncreasing=5}
    boneToModify {boneName=base_cereal_ring_3}
    rotation {pitch=-25}
    translation {x=1.002, z=2.23}
   }
   n15 = "Transform (Modify) Bone - Bone: base_cereal_ring_4"!(ComponentPose=@n14, bAlphaBoolEnabled:bool=True, Translation="0.389,-2.836,1.104", Rotation="-20,0,0", Scale="1,1,1") -> (Pose) {
    alphaInputType=Bool
    rotationMode=BMM_Additive
    rotationSpace=BCS_BoneSpace
    translationMode=BMM_Additive
    translationSpace=BCS_BoneSpace
    alphaBoolBlend {blendInTime=1, blendOutTime=0.1}
    alphaScaleBiasClamp {bInterpResult=true, interpSpeedDecreasing=5, interpSpeedIncreasing=5}
    boneToModify {boneName=base_cereal_ring_4}
    rotation {pitch=-20}
    translation=<0.389 -2.836 1.104>
   }
   n16 = "Transform (Modify) Bone - Bone: base_cereal_ring_5"!(ComponentPose=@n15, bAlphaBoolEnabled:bool=True, Translation="2,-2,0", Rotation="-30,0,0", Scale="1,1,1") -> (Pose) {
    alphaInputType=Bool
    rotationMode=BMM_Additive
    rotationSpace=BCS_BoneSpace
    translationMode=BMM_Additive
    translationSpace=BCS_BoneSpace
    alphaBoolBlend {blendInTime=1, blendOutTime=0.1}
    alphaScaleBiasClamp {bInterpResult=true, interpSpeedDecreasing=5, interpSpeedIncreasing=5}
    boneToModify {boneName=base_cereal_ring_5}
    rotation {pitch=-30}
    translation=<2 -2>
   }
   n17 = "Transform (Modify) Bone - Bone: base_cereal_ring_6"!(ComponentPose=@n16, bAlphaBoolEnabled:bool=True, Translation="5,2,-2", Rotation="-20,0,0", Scale="1,1,1") -> (Pose) {
    alphaInputType=Bool
    rotationMode=BMM_Additive
    rotationSpace=BCS_BoneSpace
    translationMode=BMM_Additive
    translationSpace=BCS_BoneSpace
    alphaBoolBlend {blendInTime=1, blendOutTime=0.1}
    alphaScaleBiasClamp {bInterpResult=true, interpSpeedDecreasing=5, interpSpeedIncreasing=5}
    boneToModify {boneName=base_cereal_ring_6}
    rotation {pitch=-20}
    translation=<5 2 -2>
   }
   n18 = "Transform (Modify) Bone - Bone: base_cereal_ring_7"!(ComponentPose=@n17, bAlphaBoolEnabled:bool=True, Translation="3.663,5.000,-0.079", Rotation="-20,0,0", Scale="1,1,1") -> (Pose) {
    alphaInputType=Bool
    rotationMode=BMM_Additive
    rotationSpace=BCS_BoneSpace
    translationMode=BMM_Additive
    translationSpace=BCS_BoneSpace
    alphaBoolBlend {blendInTime=1, blendOutTime=0.1}
    alphaScaleBiasClamp {bInterpResult=true, interpSpeedDecreasing=5, interpSpeedIncreasing=5}
    boneToModify {boneName=base_cereal_ring_7}
    rotation {pitch=-20}
    translation=<3.663 5 -0.079>
   }
   n19 = "Transform (Modify) Bone - Bone: base_cereal_ring_8"!(ComponentPose=@n18, bAlphaBoolEnabled:bool=True, Translation="1.102,2.511,-4.677", Rotation="0,0,0", Scale="1,1,1") -> (Pose) {
    alphaInputType=Bool
    rotationMode=BMM_Additive
    rotationSpace=BCS_BoneSpace
    translationMode=BMM_Additive
    translationSpace=BCS_BoneSpace
    alphaBoolBlend {blendInTime=1, blendOutTime=0.1}
    alphaScaleBiasClamp {bInterpResult=true, interpSpeedDecreasing=5, interpSpeedIncreasing=5}
    boneToModify {boneName=base_cereal_ring_8}
    translation=<1.102 2.511 -4.677>
   }
   n20 = "Transform (Modify) Bone - Bone: base_cereal_ring_9"!(ComponentPose=@n19, bAlphaBoolEnabled:bool=True, Translation="1.284,-1.550,-5.000", Rotation="0,5,0", Scale="1,1,1") -> (Pose) {
    alphaInputType=Bool
    rotationMode=BMM_Additive
    rotationSpace=BCS_BoneSpace
    translationMode=BMM_Additive
    translationSpace=BCS_BoneSpace
    alphaBoolBlend {blendInTime=1, blendOutTime=0.1}
    alphaScaleBiasClamp {bInterpResult=true, interpSpeedDecreasing=5, interpSpeedIncreasing=5}
    boneToModify {boneName=base_cereal_ring_9}
    rotation {yaw=5}
    translation=<1.284 -1.55 -5>
   }
   n21 = "Transform (Modify) Bone - Bone: base_cereal_ring_1"(ComponentPose=@n20, bAlphaBoolEnabled=True, Translation="2,0,0", Rotation="0,0,0", Scale="1,1,1") {
    alphaInputType=Bool
    rotationMode=BMM_Additive
    rotationSpace=BCS_BoneSpace
    translationMode=BMM_Additive
    translationSpace=BCS_BoneSpace
    alphaBoolBlend {blendInTime=0.1, blendOutTime=0.1}
    alphaScaleBiasClamp {interpSpeedDecreasing=5, interpSpeedIncreasing=5}
    boneToModify {boneName=base_cereal_ring_1}
    translation {x=2}
   }
   n22 = "Transform (Modify) Bone - Bone: base_cereal_ring_2"(ComponentPose=@n21, bAlphaBoolEnabled=True, Translation="2,0,0", Rotation="0,0,0", Scale="1,1,1") {
    alphaInputType=Bool
    rotationMode=BMM_Additive
    rotationSpace=BCS_BoneSpace
    translationMode=BMM_Additive
    translationSpace=BCS_BoneSpace
    alphaBoolBlend {blendInTime=0.1, blendOutTime=0.1}
    alphaScaleBiasClamp {interpSpeedDecreasing=5, interpSpeedIncreasing=5}
    boneToModify {boneName=base_cereal_ring_2}
    translation {x=2}
   }
   n23 = "Transform (Modify) Bone - Bone: base_cereal_ring_3"(ComponentPose=@n22, bAlphaBoolEnabled=True, Translation="1,0,0", Rotation="0,0,0", Scale="1,1,1") {
    alphaInputType=Bool
    rotationMode=BMM_Additive
    rotationSpace=BCS_BoneSpace
    translationMode=BMM_Additive
    translationSpace=BCS_BoneSpace
    alphaBoolBlend {blendInTime=0.1, blendOutTime=0.1}
    alphaScaleBiasClamp {interpSpeedDecreasing=5, interpSpeedIncreasing=5}
    boneToModify {boneName=base_cereal_ring_3}
    translation {x=1}
   }
   n24 = "Transform (Modify) Bone - Bone: base_cereal_ring_4"(ComponentPose=@n23, bAlphaBoolEnabled=True, Translation="1,0,0", Rotation="0,0,0", Scale="1,1,1") {
    alphaInputType=Bool
    rotationMode=BMM_Additive
    rotationSpace=BCS_BoneSpace
    translationMode=BMM_Additive
    translationSpace=BCS_BoneSpace
    alphaBoolBlend {blendInTime=0.1, blendOutTime=0.1}
    alphaScaleBiasClamp {interpSpeedDecreasing=5, interpSpeedIncreasing=5}
    boneToModify {boneName=base_cereal_ring_4}
    translation {x=1}
   }
   n25 = "Transform (Modify) Bone - Bone: base_cereal_ring_5"(ComponentPose=@n24, bAlphaBoolEnabled=True, Translation="1,0,0", Rotation="0,0,0", Scale="1,1,1") {
    alphaInputType=Bool
    rotationMode=BMM_Additive
    rotationSpace=BCS_BoneSpace
    translationMode=BMM_Additive
    translationSpace=BCS_BoneSpace
    alphaBoolBlend {blendInTime=0.1, blendOutTime=0.1}
    alphaScaleBiasClamp {interpSpeedDecreasing=5, interpSpeedIncreasing=5}
    boneToModify {boneName=base_cereal_ring_5}
    translation {x=1}
   }
   n26 = "Transform (Modify) Bone - Bone: base_cereal_ring_6"(ComponentPose=@n25, bAlphaBoolEnabled=True, Translation="1,0,0", Rotation="0,0,0", Scale="1,1,1") {
    alphaInputType=Bool
    rotationMode=BMM_Additive
    rotationSpace=BCS_BoneSpace
    translationMode=BMM_Additive
    translationSpace=BCS_BoneSpace
    alphaBoolBlend {blendInTime=0.1, blendOutTime=0.1}
    alphaScaleBiasClamp {interpSpeedDecreasing=5, interpSpeedIncreasing=5}
    boneToModify {boneName=base_cereal_ring_6}
    translation {x=1}
   }
   n27 = "Transform (Modify) Bone - Bone: base_cereal_ring_7"(ComponentPose=@n26, bAlphaBoolEnabled=True, Translation="1,0,0", Rotation="0,0,0", Scale="1,1,1") {
    alphaInputType=Bool
    rotationMode=BMM_Additive
    rotationSpace=BCS_BoneSpace
    translationMode=BMM_Additive
    translationSpace=BCS_BoneSpace
    alphaBoolBlend {blendInTime=0.1, blendOutTime=0.1}
    alphaScaleBiasClamp {interpSpeedDecreasing=5, interpSpeedIncreasing=5}
    boneToModify {boneName=base_cereal_ring_7}
    translation {x=1}
   }
   n28 = "Transform (Modify) Bone - Bone: base_cereal_ring_8"(ComponentPose=@n27, bAlphaBoolEnabled=True, Translation="2,0,0", Rotation="0,0,0", Scale="1,1,1") {
    alphaInputType=Bool
    rotationMode=BMM_Additive
    rotationSpace=BCS_BoneSpace
    translationMode=BMM_Additive
    translationSpace=BCS_BoneSpace
    alphaBoolBlend {blendInTime=0.1, blendOutTime=0.1}
    alphaScaleBiasClamp {interpSpeedDecreasing=5, interpSpeedIncreasing=5}
    boneToModify {boneName=base_cereal_ring_8}
    translation {x=2}
   }
   n29 = "Transform (Modify) Bone - Bone: base_cereal_ring_9"(ComponentPose=@n28, bAlphaBoolEnabled=True, Translation="2,0,0", Rotation="0,0,0", Scale="1,1,1") {
    alphaInputType=Bool
    rotationMode=BMM_Additive
    rotationSpace=BCS_BoneSpace
    translationMode=BMM_Additive
    translationSpace=BCS_BoneSpace
    alphaBoolBlend {blendInTime=0.1, blendOutTime=0.1}
    alphaScaleBiasClamp {interpSpeedDecreasing=5, interpSpeedIncreasing=5}
    boneToModify {boneName=base_cereal_ring_9}
    translation {x=2}
   }
   n30 = "Transform (Modify) Bone - Bone: dyn_cereal_ring_end_1"!(ComponentPose=@n29, bAlphaBoolEnabled:bool=True, Translation="0,0,0", Rotation="50,0,0", Scale="1,1,1") -> (Pose) {
    alphaInputType=Bool
    rotationMode=BMM_Additive
    rotationSpace=BCS_BoneSpace
    translationSpace=BCS_BoneSpace
    alphaBoolBlend {blendInTime=0.1, blendOutTime=0.1}
    alphaScaleBiasClamp {interpSpeedDecreasing=5, interpSpeedIncreasing=5}
    boneToModify {boneName=dyn_cereal_ring_end_1}
    rotation {pitch=50}
   }
   n31 = "Transform (Modify) Bone - Bone: dyn_cereal_ring_end_2"!(ComponentPose=@n30, bAlphaBoolEnabled:bool=True, Translation="0,0,0", Rotation="30,0,0", Scale="1,1,1") -> (Pose) {
    alphaInputType=Bool
    rotationMode=BMM_Additive
    rotationSpace=BCS_BoneSpace
    translationSpace=BCS_BoneSpace
    alphaBoolBlend {blendInTime=0.1, blendOutTime=0.1}
    alphaScaleBiasClamp {interpSpeedDecreasing=5, interpSpeedIncreasing=5}
    boneToModify {boneName=dyn_cereal_ring_end_2}
    rotation {pitch=30}
   }
   n32 = "Transform (Modify) Bone - Bone: dyn_cereal_ring_end_3"!(ComponentPose=@n31, bAlphaBoolEnabled:bool=True, Translation="0,0,0", Rotation="20,0,0", Scale="1,1,1") -> (Pose) {
    alphaInputType=Bool
    rotationMode=BMM_Additive
    rotationSpace=BCS_BoneSpace
    translationSpace=BCS_BoneSpace
    alphaBoolBlend {blendInTime=0.1, blendOutTime=0.1}
    alphaScaleBiasClamp {interpSpeedDecreasing=5, interpSpeedIncreasing=5}
    boneToModify {boneName=dyn_cereal_ring_end_3}
    rotation {pitch=20}
   }
   n33 = "Transform (Modify) Bone - Bone: dyn_cereal_ring_end_4"!(ComponentPose=@n32, bAlphaBoolEnabled:bool=True, Translation="1,0,0", Rotation="30,0,0", Scale="1,1,1") -> (Pose) {
    alphaInputType=Bool
    rotationMode=BMM_Additive
    rotationSpace=BCS_BoneSpace
    translationSpace=BCS_BoneSpace
    alphaBoolBlend {blendInTime=0.1, blendOutTime=0.1}
    alphaScaleBiasClamp {interpSpeedDecreasing=5, interpSpeedIncreasing=5}
    boneToModify {boneName=dyn_cereal_ring_end_4}
    rotation {pitch=30}
    translation {x=1}
   }
   n34 = "Transform (Modify) Bone - Bone: dyn_cereal_ring_end_5"!(ComponentPose=@n33, bAlphaBoolEnabled:bool=True, Translation="0,0,0", Rotation="10,0,0", Scale="1,1,1") -> (Pose) {
    alphaInputType=Bool
    rotationMode=BMM_Additive
    rotationSpace=BCS_BoneSpace
    translationSpace=BCS_BoneSpace
    alphaBoolBlend {blendInTime=0.1, blendOutTime=0.1}
    alphaScaleBiasClamp {interpSpeedDecreasing=5, interpSpeedIncreasing=5}
    boneToModify {boneName=dyn_cereal_ring_end_5}
    rotation {pitch=10}
   }
   n35 = "Transform (Modify) Bone - Bone: dyn_cereal_ring_end_6"!(ComponentPose=@n34, bAlphaBoolEnabled:bool=True, Translation="0,0,0", Rotation="0,0,0", Scale="1,1,1") -> (Pose) {
    alphaInputType=Bool
    rotationMode=BMM_Additive
    rotationSpace=BCS_BoneSpace
    translationSpace=BCS_BoneSpace
    alphaBoolBlend {blendInTime=0.1, blendOutTime=0.1}
    alphaScaleBiasClamp {interpSpeedDecreasing=5, interpSpeedIncreasing=5}
    boneToModify {boneName=dyn_cereal_ring_end_6}
   }
   n36 = "Transform (Modify) Bone - Bone: dyn_cereal_ring_end_7"!(ComponentPose=@n35, bAlphaBoolEnabled:bool=True, Translation="0,0,0", Rotation="5,0,0", Scale="1,1,1") -> (Pose) {
    alphaInputType=Bool
    rotationMode=BMM_Additive
    rotationSpace=BCS_BoneSpace
    translationSpace=BCS_BoneSpace
    alphaBoolBlend {blendInTime=0.1, blendOutTime=0.1}
    alphaScaleBiasClamp {interpSpeedDecreasing=5, interpSpeedIncreasing=5}
    boneToModify {boneName=dyn_cereal_ring_end_7}
    rotation {pitch=5}
   }
   n37 = "Transform (Modify) Bone - Bone: dyn_cereal_ring_end_8"!(ComponentPose=@n36, bAlphaBoolEnabled:bool=True, Translation="0,0,0", Rotation="-3,0,0", Scale="1,1,1") -> (Pose) {
    alphaInputType=Bool
    rotationMode=BMM_Additive
    rotationSpace=BCS_BoneSpace
    translationSpace=BCS_BoneSpace
    alphaBoolBlend {blendInTime=0.1, blendOutTime=0.1}
    alphaScaleBiasClamp {interpSpeedDecreasing=5, interpSpeedIncreasing=5}
    boneToModify {boneName=dyn_cereal_ring_end_8}
    rotation {pitch=-3}
   }
   n38 = "Transform (Modify) Bone - Bone: dyn_cereal_ring_end_9"!(ComponentPose=@n37, bAlphaBoolEnabled:bool=True, Translation="0,0,0", Rotation="0,0,0", Scale="1,1,1") -> (Pose) {
    alphaInputType=Bool
    rotationMode=BMM_Additive
    rotationSpace=BCS_BoneSpace
    translationSpace=BCS_BoneSpace
    alphaBoolBlend {blendInTime=0.1, blendOutTime=0.1}
    alphaScaleBiasClamp {interpSpeedDecreasing=5, interpSpeedIncreasing=5}
    boneToModify {boneName=dyn_cereal_ring_end_9}
   }
   n39 = "Transform (Modify) Bone - Bone: base_cereal_ring_1"!(ComponentPose=@n38, Alpha:float=1, Translation="0,0,0", Rotation="0,0,0", Scale="1,1,1") -> (Pose) {
    rotationMode=BMM_Additive
    rotationSpace=BCS_BoneSpace
    translationMode=BMM_Additive
    translationSpace=BCS_BoneSpace
    alphaBoolBlend {blendInTime=1, blendOutTime=0.1}
    alphaScaleBiasClamp {bInterpResult=true, interpSpeedDecreasing=5, interpSpeedIncreasing=5}
    boneToModify {boneName=base_cereal_ring_1}
   }
   n40 = "Transform (Modify) Bone - Bone: base_cereal_ring_2"!(ComponentPose=@n39, Alpha:float=1, Translation="0,0,0", Rotation="0,0,0", Scale="1,1,1") -> (Pose) {
    rotationMode=BMM_Additive
    rotationSpace=BCS_BoneSpace
    translationMode=BMM_Additive
    translationSpace=BCS_BoneSpace
    alphaBoolBlend {blendInTime=1, blendOutTime=0.1}
    alphaScaleBiasClamp {bInterpResult=true, interpSpeedDecreasing=5, interpSpeedIncreasing=5}
    boneToModify {boneName=base_cereal_ring_2}
   }
   n41 = "Transform (Modify) Bone - Bone: base_cereal_ring_3"!(ComponentPose=@n40, Alpha:float=1, Translation="0,0,0", Rotation="0,0,0", Scale="1,1,1") -> (Pose) {
    rotationMode=BMM_Additive
    rotationSpace=BCS_BoneSpace
    translationMode=BMM_Additive
    translationSpace=BCS_BoneSpace
    alphaBoolBlend {blendInTime=1, blendOutTime=0.1}
    alphaScaleBiasClamp {bInterpResult=true, interpSpeedDecreasing=5, interpSpeedIncreasing=5}
    boneToModify {boneName=base_cereal_ring_3}
   }
   n42 = "Transform (Modify) Bone - Bone: base_cereal_ring_4"!(ComponentPose=@n41, Alpha:float=1, Translation="0,0,0", Rotation="0,0,0", Scale="1,1,1") -> (Pose) {
    rotationMode=BMM_Additive
    rotationSpace=BCS_BoneSpace
    translationMode=BMM_Additive
    translationSpace=BCS_BoneSpace
    alphaBoolBlend {blendInTime=1, blendOutTime=0.1}
    alphaScaleBiasClamp {bInterpResult=true, interpSpeedDecreasing=5, interpSpeedIncreasing=5}
    boneToModify {boneName=base_cereal_ring_4}
   }
   n43 = "Transform (Modify) Bone - Bone: base_cereal_ring_5"!(ComponentPose=@n42, Alpha:float=1, Translation="0,0,0", Rotation="0,0,0", Scale="1,1,1") -> (Pose) {
    rotationMode=BMM_Additive
    rotationSpace=BCS_BoneSpace
    translationMode=BMM_Additive
    translationSpace=BCS_BoneSpace
    alphaBoolBlend {blendInTime=1, blendOutTime=0.1}
    alphaScaleBiasClamp {bInterpResult=true, interpSpeedDecreasing=5, interpSpeedIncreasing=5}
    boneToModify {boneName=base_cereal_ring_5}
   }
   n44 = "Transform (Modify) Bone - Bone: base_cereal_ring_6"!(ComponentPose=@n43, Alpha:float=1, Translation="0,0,0", Rotation="0,0,0", Scale="1,1,1") -> (Pose) {
    rotationMode=BMM_Additive
    rotationSpace=BCS_BoneSpace
    translationMode=BMM_Additive
    translationSpace=BCS_BoneSpace
    alphaBoolBlend {blendInTime=1, blendOutTime=0.1}
    alphaScaleBiasClamp {bInterpResult=true, interpSpeedDecreasing=5, interpSpeedIncreasing=5}
    boneToModify {boneName=base_cereal_ring_6}
   }
   n45 = "Transform (Modify) Bone - Bone: base_cereal_ring_7"!(ComponentPose=@n44, Alpha:float=1, Translation="0,0,0", Rotation="0,0,0", Scale="1,1,1") -> (Pose) {
    rotationMode=BMM_Additive
    rotationSpace=BCS_BoneSpace
    translationMode=BMM_Additive
    translationSpace=BCS_BoneSpace
    alphaBoolBlend {blendInTime=1, blendOutTime=0.1}
    alphaScaleBiasClamp {bInterpResult=true, interpSpeedDecreasing=5, interpSpeedIncreasing=5}
    boneToModify {boneName=base_cereal_ring_7}
   }
   n46 = "Transform (Modify) Bone - Bone: base_cereal_ring_8"!(ComponentPose=@n45, Alpha:float=1, Translation="0,0,0", Rotation="0,0,0", Scale="1,1,1") -> (Pose) {
    rotationMode=BMM_Additive
    rotationSpace=BCS_BoneSpace
    translationMode=BMM_Additive
    translationSpace=BCS_BoneSpace
    alphaBoolBlend {blendInTime=1, blendOutTime=0.1}
    alphaScaleBiasClamp {bInterpResult=true, interpSpeedDecreasing=5, interpSpeedIncreasing=5}
    boneToModify {boneName=base_cereal_ring_8}
   }
   n47 = "Transform (Modify) Bone - Bone: base_cereal_ring_9"!(ComponentPose=@n46, Alpha:float=1, Translation="0,0,0", Rotation="0,0,0", Scale="1,1,1") -> (Pose) {
    rotationMode=BMM_Additive
    rotationSpace=BCS_BoneSpace
    translationMode=BMM_Additive
    translationSpace=BCS_BoneSpace
    alphaBoolBlend {blendInTime=1, blendOutTime=0.1}
    alphaScaleBiasClamp {bInterpResult=true, interpSpeedDecreasing=5, interpSpeedIncreasing=5}
    boneToModify {boneName=base_cereal_ring_9}
   }
   Save cached pose 'CR_Cache_Transforms'(Pose=Component To Local(ComponentPose=@n47)) {cacheName=CR_Cache_Transforms}
  ]}
 EventGraph {nodes [
   Event BlueprintUpdateAnimation()
   > n1 = Sequence!(execute) -> (then_0, then_1, then_2:Exec)
   > n2 = Branch(Condition=Get bIsInFrontEnd())
   > Set isFrontEnd(isFrontEnd:double=1) -> (Output_Get:double)
   Set isNotFrontEnd(execute=@n2.else, isNotFrontEnd:double=1) -> (Output_Get:double)
   Set Sim Space Settings(
    execute=@n1.then_1
    Sim Space Settings:Anim Phys Sim Space Settings Structure=Get Sim Space Settings_BR() -> (Sim Space Settings_BR:Anim Phys Sim Space Settings Structure)
   ) -> (Output_Get:Anim Phys Sim Space Settings Structure)
   > Branch(Condition=Get bIsInFrontEnd())
   > Set Sim Space Settings(Sim Space Settings=Get Sim Space Settings_FE() -> (Sim Space Settings_FE:Anim Phys Sim Space Settings Structure))
   n3 = GetRealTimeSeconds()
   n4 = "Sin (Radians)"(A=float * float(A=float * float(A=0.9, B=1.1), B=@n3))
   n5 = "Sin (Radians)"(A=float * float(A=float * float(A=1.1, B=1.1), B=@n3))
   n6 = "Sin (Radians)"(A=float * float(A=@n3, B=float * float(A=1.2, B=5)))
   Set RingRotation1(
    execute=@n1.then_2
    RingRotation1:Rotator=MakeRotator(
     Roll=float * float(A=@n5, B=-2)
     Pitch=float * float(A=@n4, B=3)
     Yaw=Remap Float(Input=@n6, OldRange=(X=-1,Y=1), NewRange=(X=-0.2,Y=0.2)) -> (Output=~)
    )
   ) -> (Output_Get:Rotator)
   n7 = GetRealTimeSeconds()
   n8 = "Sin (Radians)"(A=float * float(A=float * float(A=0.9, B=1.1), B=@n7))
   n9 = "Sin (Radians)"(A=float * float(A=float * float(A=1.1, B=1.1), B=@n7))
   n10 = "Sin (Radians)"(A=float * float(A=@n7, B=float * float(A=1.2, B=5)))
   > Set RingRotation3(RingRotation3:Rotator=MakeRotator(
     Roll=float * float(A=@n9, B=2)
     Pitch=float * float(A=@n8, B=-1.5)
     Yaw=Remap Float(Input=@n10, OldRange=(X=-1,Y=1), NewRange=(X=-0.25,Y=0.25)) -> (Output=~)
    )) -> (Output_Get:Rotator)
   n11 = GetRealTimeSeconds()
   n12 = "Sin (Radians)"(A=float * float(A=float * float(A=0.9, B=1.1), B=@n11))
   n13 = "Sin (Radians)"(A=float * float(A=float * float(A=1.1, B=1.1), B=@n11))
   n14 = "Sin (Radians)"(A=float * float(A=@n11, B=float * float(A=1.2, B=5)))
   > Set RingRotation2(RingRotation2:Rotator=MakeRotator(
     Roll=float * float(A=@n13, B=2)
     Pitch=float * float(A=@n12, B=4.5)
     Yaw=Remap Float(Input=@n14, OldRange=(X=-1,Y=1), NewRange=(X=-0.33,Y=0.33)) -> (Output=~)
    )) -> (Output_Get:Rotator)
   n15 = GetRealTimeSeconds()
   n16 = "Sin (Radians)"(A=float * float(A=float * float(A=0.9, B=1.1), B=@n15))
   n17 = "Sin (Radians)"(A=float * float(A=float * float(A=1.1, B=1.1), B=@n15))
   n18 = "Sin (Radians)"(A=float * float(A=@n15, B=float * float(A=1.2, B=5)))
   > Set RingRotation4(RingRotation4:Rotator=MakeRotator(
     Roll=float * float(A=@n17, B=-1.2)
     Pitch=float * float(A=@n16, B=-0.5)
     Yaw=Remap Float(Input=@n18, OldRange=(X=-1,Y=1), NewRange=(X=-0.3,Y=0.3)) -> (Output=~)
    )) -> (Output_Get:Rotator)
   Event OnBeginJumping()
   > Set Is Jumping(Is Jumping:bool=1) -> (Output_Get:bool)
   Event OnEndFalling()
   > Set Is Jumping(Is Jumping=false)
   TryGetPawnOwner()
  ]}
]
variables [
 Is Jumping {type=boolean}
 isFrontEnd {type=number}
 isNotFrontEnd {type=number}
 RingRotation1 {type=Rotator}
 RingRotation2 {type=Rotator}
 RingRotation3 {type=Rotator}
 RingRotation4 {type=Rotator}
 Sim Space Settings {
  type=AnimPhysSimSpaceSettings
  default {maxAngularAcceleration=10000, maxAngularVelocity=10000, simSpaceAngularAlpha=0, externalAngularVelocity=<0 0 0>}
 }
 Sim Space Settings_BR {
  type=AnimPhysSimSpaceSettings
  default {maxAngularAcceleration=10000, maxAngularVelocity=10000, simSpaceAngularAlpha=0, externalAngularVelocity=<0 0 0>}
 }
 Sim Space Settings_FE {
  type=AnimPhysSimSpaceSettings
  default {maxAngularAcceleration=10000, maxAngularVelocity=10000, simSpaceAngularAlpha=0.1, externalAngularVelocity=<0.1 0.1 0.1>}
 }
]
